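; Prototype searchable loot container Blueprint for the Athena Flopper, named as a tiered ammo container.
; On BeginPlay it makes its mesh a movable PhysicsBody that overlaps two custom trace channels.
; Mobile highlight events create a yaw-countered interact icon, scale it by timeline, and reverse that on unhighlight or search to remove it.
; Looting animates the material's CODE_Timestamp parameter by timeline, or snaps it to 1 on lower rendering detail.
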(event EventBeginPlay
  (|CallParentFunction)
  (Transformation|SetMobility (Variables|Debug|GetStaticMeshComponent) "Movable")
  (Collision|SetCollisionObjectType (Variables|Debug|GetStaticMeshComponent) "ECC_PhysicsBody")
  (Collision|SetCollisionResponsetoChannel (Variables|Debug|GetStaticMeshComponent) "ECC_GameTraceChannel15" "ECR_Overlap")
  (Collision|SetCollisionResponsetoChannel (Variables|Debug|GetStaticMeshComponent) "ECC_GameTraceChannel18" "ECR_Overlap"))

(event Container|EventEnableBacchusHighlight
  (bind (_returnvalue_roll _returnvalue_pitch _returnvalue_yaw) (Math|Rotator|BreakRotator (Transformation|GetActorRotation)))
  ;[1 1 1] Mobile highlighting
  (Networking|CanExecuteCosmeticEvents
    (:True
      (Utilities|SetVisibleMobileIcon true (Variables|Default|GetMobileInteractIcon) (Variables|Default|GetMobileInteractionMaterial) self (Variables|Default|GetMobileInteractIconLocation) (Math|Rotator|MakeRotator 0.0 0.0 (* _returnvalue_yaw -1.0)) (Variables|Default|GetMobileInteractIconScale)
        (:Out
          (bind (_direction _lerpobject _lerpinteactoicon) (Timeline|MobileSelectedTL Play)
            (:Update
              (bind _mobileinteracticon (Variables|Default|GetMobileInteractIcon)
                (:then
                  (Transformation|SetRelativeScale3D _mobileinteracticon (Math|Vector|Lerp(Vector) "0, 0, 0" (Variables|Default|GetMobileInteractIconScale) _lerpinteactoicon)))
                (:else)))
            (:Finished
              (switch Utilities|FlowControl|Switch|SwitchonETimelineDirection _direction
                (:Backward
                  (Utilities|SetVisibleMobileIcon :Mesh (Variables|Default|GetMobileInteractIcon) :Rotation (0.0 0.0 0.0)))))))))))

(event Container|EventDisableBacchusHighlight
  ;[1 1 1] Mobile highlighting
  (if (TouchPlatform|IsMobileGame)
    (Networking|CanExecuteCosmeticEvents
      (:True
        (Timeline|MobileSelectedTL Reverse)))))

(event EventOnBeginSearch
  ;[1 1 1] Mobile highlighting
  (if (Variables|Container|GetbAlreadySearched)
    _
    (elif (TouchPlatform|IsMobileGame)
      (Networking|CanExecuteCosmeticEvents
        (:True
          (Timeline|MobileSelectedTL Reverse))))))

(event EventOnLoot
  ;[1 1 1] Animate Chest
  (bind _returnvalue (Rendering|Material|CreateDynamicMaterialInstance (Variables|Debug|GetStaticMeshComponent) 0 0 "None" :!class /Script/Engine.PrimitiveComponent))
  (if (>= (Rendering|GetRenderingDetailMode) 3)
    (bind (_direction _time) (Timeline|Timeline_0 PlayFromStart)
      (:Update
        (Rendering|Material|SetScalarParameterValue _returnvalue "CODE_Timestamp" _time :!class /Script/Engine.MaterialInstanceDynamic)))
    (else
      (Rendering|Material|SetScalarParameterValue _returnvalue "CODE_Timestamp" 1.0 :!class /Script/Engine.MaterialInstanceDynamic))))

(fn SetVisibleMobileInteractIcon
  (params (Visible :type Boolean))
  (bind (_returnvalue_roll _returnvalue_pitch _returnvalue_yaw) (Math|Rotator|BreakRotator (Transformation|GetActorRotation)))
  (if Visible
    ;[1 1 1] Create
    (bind _mobileinteracticon (Variables|Default|GetMobileInteractIcon)
      (:then
        (Rendering|SetVisibility _mobileinteracticon true false))
      (:else
        (Variables|Default|GetMobileInteractionMaterial
          (:then
            (bind _returnvalue (AddComponent|NODE_AddStaticMeshComponent-0 :RelativeTransform_Location (Variables|Default|GetMobileInteractIconLocation) :RelativeTransform_Rotation_Yaw (* _returnvalue_yaw -1.0) :RelativeTransform_Scale (Variables|Default|GetMobileInteractIconScale)))
            (bind _output_get (Variables|Default|SetMobileInteractIcon _returnvalue))
            (Rendering|Material|SetMaterial _output_get 0 (Variables|Default|GetMobileInteractionMaterial)))
          (:else))))
    (else
      ;[1 1 1] Destroy
      (bind _mobileinteracticon_2 (Variables|Default|GetMobileInteractIcon)
        (:then
          (Rendering|SetVisibility _mobileinteracticon_2 false false)
          (Components|DestroyComponent _mobileinteracticon_2))
        (:else)))))

(fn ConstructionScript)
