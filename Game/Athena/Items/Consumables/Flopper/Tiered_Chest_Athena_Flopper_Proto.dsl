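; Prototype Flopper tiered chest, a child Blueprint that adapts the chest to move via a FortProjectileMovement component.
; On BeginPlay, after the parent logic, it makes its static mesh movable, sets it to the PhysicsBody object type and sets two custom game collision channels to overlap.
; A custom ActivateMovement event activates the movement component with a reset; nothing in this file calls it.
; The construction script only runs the parent's.
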(event EventBeginPlay
  (|CallParentFunction)
  (Transformation|SetMobility (Variables|Debug|GetStaticMeshComponent) "Movable")
  (Collision|SetCollisionObjectType (Variables|Debug|GetStaticMeshComponent) "ECC_PhysicsBody")
  (Collision|SetCollisionResponsetoChannel (Variables|Debug|GetStaticMeshComponent) "ECC_GameTraceChannel15" "ECR_Overlap")
  (Collision|SetCollisionResponsetoChannel (Variables|Debug|GetStaticMeshComponent) "ECC_GameTraceChannel18" "ECR_Overlap"))

(event Custom|ActivateMovement
  (Components|Activation|Activate (Variables|Default|GetFortProjectileMovement) true))

(fn ConstructionScript
  (|CallParentFunction))
